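# M_BugSwarm_Beacon_Spiral is a material in the environment bug effects folder, named for a spiral bug-swarm beacon effect.
# Its expression graph holds no nodes, so the file shows no textures, parameters or animation that would produce the spiral.
# It uses the Default Lit shading model and has automatic usage flag setting in the editor turned off.
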Material /Game/Effects/Fort_Effects/Effects/Environment/Bugs/M_BugSwarm_Beacon_Spiral
graph {expressions {}, properties {bAutomaticallySetUsageInEditor=false, usedShadingModels=Default Lit}}
